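; Character part modifier for the fire Demon Mask that drives its equip, hand-flame and toss visuals on the player pawn.
; When customization finishes, it attaches Niagara effects to the mask and middle fingers, a noted log-spam workaround.
; B_DemonMask_Fire ammo events switch hand flames, while prop-hide and cloak tag tasks toggle effect visibility.
; Game-time stamps in custom primitive data feed the mask material on equip and toss, with delayed hiding after a toss.
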(event Custom|Mask%20NS
  (bind _mask_equip_ns (Variables|Default|GetMask%20Equip%20NS))
  ;[0.4383 0.1161 1] TODO
  ;;[1 1 1] This is a work around for FORT-589563. A systemic fix described in FORT-617814 should make the original code work without log spam. Note also this work around required the addition of the Glove Equip NS component.
  (Components|Activation|Activate _mask_equip_ns false)
  (Transformation|AttachComponentToComponent _mask_equip_ns (CustomCharacter|GetSkeletalMeshForPartOnPlayerPawn) "offset" "KeepRelative" "KeepRelative" "KeepRelative" false)
  (Transformation|AttachComponentToComponent (Variables|Default|GetNS_DemonMask_Fire_IdleHands_L) (Variables|/Script/Engine.Character|GetMesh (CustomCharacter|GetAssociatedPlayerPawn)) "middle_01_l" "KeepRelative" "KeepRelative" "KeepRelative" true)
  (Transformation|AttachComponentToComponent (Variables|Default|GetNS_DemonMask_Fire_IdleHands_R) (Variables|/Script/Engine.Character|GetMesh (CustomCharacter|GetAssociatedPlayerPawn)) "middle_01_r" "KeepRelative" "KeepRelative" "KeepRelative" true))

(event EventOnPostResIn
  (bind _prophidetag (Variables|Default|GetPropHideTag))
  (bind _mask_equip_ns (Variables|Default|GetMask%20Equip%20NS))
  (bind _mask_equip_ns_2 (Variables|Default|GetMask%20Equip%20NS))
  (bind _as_b_demon_mask_fire (Variables|Default|GetAs%20B%20Demon%20Mask%20Fire))
  (bind _ns_demonmask_fire_idlehands_l (Variables|Default|GetNS_DemonMask_Fire_IdleHands_L))
  (|CallParentFunction)
  (bind _asfort_player_pawn_athena (Utilities|Casting|CastToFortPlayerPawnAthena (CustomCharacter|GetAssociatedPlayerPawn)))
  (bind _asb_demon_mask_fire (Utilities|Casting|CastToB_DemonMask_Fire (Variables|/Script/FortniteGame.FortPawn|GetCurrentWeapon _asfort_player_pawn_athena)))
  (bind _output_get (Variables|Default|SetAs%20B%20Demon%20Mask%20Fire _asb_demon_mask_fire))
  (Variables|/DemonMaskFire/Gameplay/Actors/B_DemonMask_Fire.B_DemonMask_Fire_C|SetPartModifier self _output_get)
  ;[0 0.07551 0.901 0.5] Ammo tracking for hand flames
  (EventDispatchers|BindEventtoOnConsumeAmmo _as_b_demon_mask_fire (EventDispatchers|CreateEvent :!event Custom|OnConsumeAmmo))
  (EventDispatchers|BindEventtoOnLocalAmmoChanged _as_b_demon_mask_fire (EventDispatchers|CreateEvent :!event Custom|OnLocalAmmoChanged))
  (bind _output_get_2 (Variables|Default|SetLocal%20Mag%20Ammo (FortWeapon|GetMagazineAmmoCount _as_b_demon_mask_fire)))
  (Variables|Default|SetPrevious%20Local%20Mag%20Ammo _output_get_2)
  ; Hide Effects while in Prop
  (Ability|Tasks|WaitGameplayTagAddToActor _asfort_player_pawn_athena _prophidetag
    (:then
      (Ability|Tasks|WaitGameplayTagRemoveFromActor _asfort_player_pawn_athena _prophidetag
        (:then
          ; Turn invisible if the player goes invisible
          (Ability|Tasks|WaitGameplayTagAddToActor _asfort_player_pawn_athena (Variables|Default|GetCloakTag)
            (:then
              (Ability|Tasks|WaitGameplayTagRemoveFromActor _asfort_player_pawn_athena (Variables|Default|GetCloakTag)
                (:Removed
                  (Rendering|SetVisibility _mask_equip_ns_2 true true)
                  (Rendering|SetVisibility _mask_equip_ns_2 true true)
                  (Rendering|SetVisibility (Variables|Default|GetNS_DemonMask_Fire_IdleHands_R) true true))))
            (:Added
              (Rendering|SetVisibility _mask_equip_ns false true)
              (Rendering|SetVisibility _mask_equip_ns false true)
              (Rendering|SetVisibility (Variables|Default|GetNS_DemonMask_Fire_IdleHands_R) false true))))
        (:Removed
          (Rendering|SetVisibility _ns_demonmask_fire_idlehands_l true true))))
    (:Added
      (Rendering|SetVisibility _ns_demonmask_fire_idlehands_l false true))))

(event EventOnDayPhaseChanged
  (params (CurrentDayPhase :type EFortDayPhase%20Enum :default "Morning") (PreviousDayPhase :type EFortDayPhase%20Enum :default "Morning") (bAtCreation :type Boolean)))

(event EventOnNewPlayerWeaponType
  (params (NewType :type EFortItemType%20Enum :default "WorldItem") (OldType :type EFortItemType%20Enum :default "WorldItem"))
  (|CallParentFunction NewType OldType) ; This node is disabled and will not be called.\nDrag off pins to build functionality.
)

(event EventOnWeaponChanged
  (params (NewWeapon :type Fort%20Weapon%20Object%20Reference) (OldWeapon :type Fort%20Weapon%20Object%20Reference))
  (bind _as_b_demon_mask_fire (Variables|Default|GetAs%20B%20Demon%20Mask%20Fire))
  ; Calling it just for when the mask is unequipped
  (|CallParentFunction NewWeapon OldWeapon)
  (Utilities|IsValid NewWeapon
    (:Is%20Valid
      (Utilities|Casting|CastToB_DemonMask_Fire NewWeapon
        (:CastFailed
          (Utilities|IsValid (CustomCharacter|GetSkeletalMeshForPartOnPlayerPawn)
            (:Is%20Valid
              (EventDispatchers|UnbindEventfromOnLocalAmmoChanged _as_b_demon_mask_fire (EventDispatchers|CreateEvent :!event Custom|OnLocalAmmoChanged))
              (EventDispatchers|UnbindEventfromOnConsumeAmmo _as_b_demon_mask_fire (EventDispatchers|CreateEvent :!event Custom|OnConsumeAmmo)))
            (:Is%20Not%20Valid
              (Development|PrintString "INVALID"))))))))

(event EventOnCurieStateDetached
  (params (Tag :type Gameplay%20Tag%20Structure%20%28by%20ref%29)))

(event EventOnWeaponFired)

(event EventOnNewWeapons
  (params (WeaponsAsItems :type Array%20of%20Fort%20Item%20Object%20References)))

(event EventOnCharacterCustomizationFinished
  (params (CharacterPartInfo :type Array%20of%20Applied%20Character%20Part%20Info%20Structures))
  (bind _returnvalue (CustomCharacter|GetSkeletalMeshForPartOnPlayerPawn))
  (|CallParentFunction CharacterPartInfo)
  (Utilities|IsValid _returnvalue
    (:Is%20Valid
      (Rendering|SetVisibility _returnvalue true true)
      (Rendering|Material|SetCustomPrimitiveDataFloat _returnvalue 0 (Utilities|Time|GetGameTimeinSeconds))
      (CallFunction|MaskNS))))

(event EventOnCurieStateAttached
  (params (Tag :type Gameplay%20Tag%20Structure%20%28by%20ref%29)))

(event EventOnBeginSkydiving
  (|CallParentFunction) ; This node is disabled and will not be called.\nDrag off pins to build functionality.
)

(event EventOnEndSkydiving
  (|CallParentFunction) ; This node is disabled and will not be called.\nDrag off pins to build functionality.
)

(event Custom|OnConsumeAmmo
  (params (Quantity :type Integer))
  ;[0 0.108 0.849 0.5] Control Hand NS based on ammo count
  (Variables|Default|SetPrevious%20Local%20Mag%20Ammo (FortWeapon|GetMagazineAmmoCount (Variables|Default|GetAs%20B%20Demon%20Mask%20Fire))))

(event EventOnEnterVehicle
  (params (VehicleEntered :type Fort%20Vehicle%20Interface%20Interface%20%28by%20ref%29))
  (|CallParentFunction VehicleEntered) ; This node is disabled and will not be called.\nDrag off pins to build functionality.
)

(event EventOnExitVehicle
  (|CallParentFunction) ; This node is disabled and will not be called.\nDrag off pins to build functionality.
)

(event Custom|OnLocalAmmoChanged
  (params (LocalCount :type Integer) (LocalRemaining :type Integer))
  ;[0 0.108 0.849 0.5] Control Hand NS based on ammo count
  (switch Utilities|FlowControl|Switch|SwitchonDemonMaskFire_AmmoCount (Utilities|Enum|BytetoEnumDemonMaskFire_AmmoCount (Math|Conversions|ToByte(Integer) LocalCount))
    (:NewEnumerator0
      (Components|Activation|Deactivate (Variables|Default|GetNS_DemonMask_Fire_IdleHands_L)))
    (:NewEnumerator1
      (if (> LocalCount (Variables|Default|GetPrevious%20Local%20Mag%20Ammo))
        (Components|Activation|Deactivate (Variables|Default|GetNS_DemonMask_Fire_IdleHands_L))
        (Components|Activation|Activate (Variables|Default|GetNS_DemonMask_Fire_IdleHands_R) false)
        (else
          (Components|Activation|Activate (Variables|Default|GetNS_DemonMask_Fire_IdleHands_R) false)
          (Components|Activation|Deactivate (Variables|Default|GetNS_DemonMask_Fire_IdleHands_L)))))
    (:NewEnumerator2
      (Components|Activation|Activate (Variables|Default|GetNS_DemonMask_Fire_IdleHands_R) false))))

(event EventOnTargetingOrFirstPersonCameraChanged)

(event EventOnHitPawn
  (params (SelfActor :type Actor%20Object%20Reference) (OtherActor :type Actor%20Object%20Reference) (NormalImpulse :type Vector :default "0, 0, 0") (Hit :type Hit%20Result%20Structure%20%28by%20ref%29)))

(event Custom|DemonMaskFire_MaterialWipeOnToss
  (bind _returnvalue (CustomCharacter|GetSkeletalMeshForPartOnPlayerPawn))
  (bind _mask_equip_ns (Variables|Default|GetMask%20Equip%20NS))
  (Utilities|IsValid _returnvalue
    (:Is%20Valid
      (Rendering|Material|SetCustomPrimitiveDataFloat _returnvalue 1 (Utilities|Time|GetGameTimeinSeconds))
      (Utilities|FlowControl|Sequence
        (:then_0
          (Utilities|FlowControl|Delay 0.45)
          (Components|Activation|Deactivate _mask_equip_ns)
          (Rendering|SetVisibility _mask_equip_ns false true))
        (:then_1
          (Utilities|FlowControl|Delay 0.7)
          (Rendering|SetVisibility (CustomCharacter|GetSkeletalMeshForPartOnPlayerPawn) false false))))))

(event Custom|CustomEvent)

(fn ConstructionScript
  (|CallParentFunction))
